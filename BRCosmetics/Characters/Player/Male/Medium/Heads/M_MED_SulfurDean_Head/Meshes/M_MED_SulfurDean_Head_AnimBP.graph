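# Animation Blueprint for the SulfurDean male player head that toggles a material element based on which character part mesh is equipped.
# When character customization loads, it caches a dynamic material instance from the part's skeletal mesh component.
# Every animation update wildcard-matches the pawn's part mesh name against M_MED_SulfurDean_Hear.
# The result drives the "Hide Element 01" scalar, hiding the element on a match and showing it otherwise.
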AnimBlueprint /BRCosmetics/Characters/Player/Male/Medium/Heads/M_MED_SulfurDean_Head/Meshes/M_MED_SulfurDean_Head_AnimBP
target_skeleton=/Game/Characters/Player/Common/Fortnite_Base_Head/Fortnite_Base_Head_Export_Skeleton
graphs [EventGraph {nodes [
   Event OnCharacterCustomizationLoaded()
   > n1 = Get Dynamic Material Instance(SkeletalMeshComponent=GetPartSkeletalMeshComponent())
   > Set MatInsta(MatInsta:Material Instance Dynamic Object Reference=@n1) -> (Output_Get:Material Instance Dynamic Object Reference)
   n2 = Event BlueprintUpdateAnimation()
   > Parent: BlueprintUpdateAnimation!(execute, DeltaTimeX:float=@n2.DeltaTimeX) -> (then) {is=title K2Node_CallParentFunction}
   > n3 = Cast To FortPlayerPawn(Object=TryGetPawnOwner())
   n4 = GetSkeletalMeshForPartType(self=@n3)
   > n5 = Set isSulfurDeanHear(isSulfurDeanHear:bool=MatchesWildcard(SourceString=GetDisplayName(Object=@n4), Wildcard=*M_MED_SulfurDean_Hear*, SearchCase=CaseSensitive)) -> (Output_Get:bool)
   > n6 = Branch(Condition=@n5)
   n7 = Get MatInsta() -> (MatInsta:Material Instance Dynamic Object Reference)
   > SetScalarParameterValue(self=@n7, ParameterName=Hide Element 01, Value=1)
   SetScalarParameterValue(execute=@n6.else, self=@n7, ParameterName=Hide Element 01, Value=0)
   TryGetPawnOwner()
  ]}]
variables [
 isSulfurDeanHear {type=boolean}
 MatInsta {type=MaterialInstanceDynamic}
]
